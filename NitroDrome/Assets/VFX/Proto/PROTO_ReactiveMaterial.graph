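# PROTO_ReactiveMaterial is a prototype translucent, Default Lit material in the NitroDrome VFX folder.
# Its only node reads the NitroDrome_Test01 parameter from the MPC_NitroDrome Material Parameter Collection.
# No connection from that node to a material output is shown, so the reactive behavior its name suggests is not built in this graph.
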Material /NitroDrome/Assets/VFX/Proto/PROTO_ReactiveMaterial
graph {
 properties {blendMode=BLEND_Translucent, pixelDepthOffsetMode=PDOM_Legacy, usedShadingModels=Default Lit}
 nodes [CollectionParameter() {collection=/Game/Environments/Helios/FX/NitroDrome/MPC_NitroDrome, parameterName=NitroDrome_Test01}]
}
